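; Zeus boss bot AI controller that adjusts the gameplay effect applied while patrolling based on the boss phase.
; AddPatrollingGE checks the bot pawn for Phase 2 gameplay tags and, if matched, applies the Phase 2 patrolling effect through its ability system component, storing the speed effect handle.
; Otherwise it defers to the parent implementation.
; BeginPlay, Tick and the construction script add no logic of their own.
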(event EventBeginPlay
  (|CallParentFunction) ; This node is disabled and will not be called.\nDrag off pins to build functionality.
)

(event EventTick
  (params (DeltaSeconds :type Float%20%28single-precision%29))
  (|CallParentFunction DeltaSeconds) ; This node is disabled and will not be called.\nDrag off pins to build functionality.
)

(fn AddPatrollingGE
  (bind _playerbotpawn (Variables|FortAthenaAIBotController|GetPlayerBotPawn))
  (bind _returnvalue (Ability|GetAbilitySystemComponent _playerbotpawn))
  (Utilities|IsValid _playerbotpawn
    (:Is%20Valid
      (if (GameplayTags|HasAnyMatchingGameplayTags _playerbotpawn (Variables|Default|GetTags_Phase2))
        (Utilities|IsValid _returnvalue
          (:Is%20Valid
            (bind _returnvalue_2 (GameplayEffects|ApplyGameplayEffectToTarget :self _returnvalue :GameplayEffectClass (Variables|Default|GetGE_Phase2Patrolling) :Target _returnvalue :Level 0.0))
            (Variables|Default|SetGE_PatrollingSpeedHandle _returnvalue_2)))
        (else
          (|CallParentFunction))))))

(fn ConstructionScript
  (|CallParentFunction))
